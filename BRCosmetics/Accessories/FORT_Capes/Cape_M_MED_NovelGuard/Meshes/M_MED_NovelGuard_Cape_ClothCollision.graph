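# Physics asset for the NovelGuard cape accessory mesh, named as its cloth collision.
# It holds a single body on the backpack_root bone, made of one offset sphere that gives the cape's cloth a volume to collide against.
# It has no constraints.
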PhysicsAsset /BRCosmetics/Accessories/FORT_Capes/Cape_M_MED_NovelGuard/Meshes/M_MED_NovelGuard_Cape_ClothCollision
bodies {backpack_root [{shapeType=Sphere, radius=14, shapeName=backpack_root_sphere, center=<16.0886 -7.3501 -22.5372>}]}
constraints []
